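; Editor Geometry Script actor that converts tree static meshes into Nanite static mesh assets.
; Rebuild and Update share duplicated logic: copy the source, optionally remap material IDs against a fallback mesh's SnowyPine materials, and merge sections using UV offsets.
; With Debug on, BakeStaticmesh creates or overwrites the target asset, writing hi-res and fallback meshes, with convex collision from CollisionMesh or a cylinder.
; InfuseCustomLOD separately writes a custom mesh into a chosen LOD.
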(event EventBeginPlay)

(event Collision|EventActorBeginOverlap
  (params (OtherActor :type Actor%20Object%20Reference)))

(event EventTick
  (params (DeltaSeconds :type Float%20%28single-precision%29)))

(event Custom|Update
  (bind _staticmesh (Variables|Default|GetStaticmesh))
  (bind _staticmaterials (Variables|/Script/Engine.StaticMesh|GetStaticMaterials _staticmesh))
  (bind _returnvalue (Utilities|Array|LastIndex _staticmaterials))
  (bind _meshmaterialid (Variables|Default|GetMeshMaterialID))
  (Utilities|FlowControl|Sequence
    (:then_0
      (bind _returnvalue_2 (DynamicMeshComponent|GetDynamicMesh (Variables|DynamicMeshActor|GetDynamicMeshComponent)))
      (bind _output_get (Variables|Default|SetDynamicMesh _returnvalue_2))
      (bind _returnvalue_3 (GeometryScript|Materials|EnableMaterialIDs _output_get))
      (bind _returnvalue_4 (GeometryScript|StaticMesh|CopyMeshFromStaticMeshwithSectionMaterials (Variables|Default|GetStaticmesh) _returnvalue_3 (Utilities|Struct|MakeGeometryScriptCopyMeshfromAssetOptions true) (Utilities|Struct|MakeGeometryScriptMeshReadLOD "MaxAvailable"))
        (:Success
          (Utilities|FlowControl|Sequence
            (:then_0
              (if (not (Utilities|IsValid (Variables|Default|GetCollisionMesh) :!class /Script/Engine.KismetSystemLibrary))
                (bind _returnvalue_5 (DynamicMeshActor|AllocateComputeMesh))
                (bind _returnvalue_6 (GeometryScript|Primitives|AppendCylinder :TargetMesh _returnvalue_5 :PrimitiveOptions (Utilities|Struct|MakeGeometryScriptPrimitiveOptions "PerFace") :Radius 5e+01 :Height 5e+02 :RadialSteps 12 :HeightSteps 0 :bCapped true :Origin "Base"))
                (Variables|Default|SetHull%20Mesh _returnvalue_6)))
            (:then_1
              (switch Utilities|FlowControl|Switch|SwitchonConvertTree_Mode (Variables|Default|GetMode)
                (:NewEnumerator0
                  (if (Variables|Default|GetHasFallbackMesh)
                    (Utilities|Array|Clear (Variables|Default|GetnewMaterialIDs))
                    (Utilities|Array|Clear (Variables|Default|GetTriangle%20IDListByMaterialID))
                    (elif (Variables|Default|GetFlipMateiralIDsForNaniteandFallback)
                      _
                      (else
                        (Utilities|FlowControl|Sequence
                          (:then_0
                            (Development|PrintString (Utilities|String|Append "GeoScript NewMatIDs =" "Start"))
                            (for _index (range (+ _returnvalue 1))
                              (Utilities|Array|Add (Variables|Default|GetTriangle%20IDListByMaterialID) (GeometryScript|Selection|GetTrianglesbyMaterialID (Variables|Default|GetDynamicMesh) _index))
                              (bind _asmaterial_instance_constant (Utilities|Casting|CastToMaterialInstanceConstant (StaticMesh|GetMaterial (Variables|Default|GetStaticmesh) _index)))
                              (bind _index_2 (Utilities|FlowControl|ForLoopwithBreak 0 _returnvalue)
                                (:LoopBody
                                  (bind _asmaterial_instance_constant_2 (Utilities|Casting|CastToMaterialInstanceConstant (StaticMesh|GetMaterial (Variables|Default|GetStaticmesh_Fallback) _index_2)))
                                  (bind (_parametersource _returnvalue_7) (MaterialEditing|GetStaticSwitchParameterSource _asmaterial_instance_constant_2 "IsTrunk"))
                                  (bind (_parametersource_2 _returnvalue_8) (MaterialEditing|GetStaticSwitchParameterSource _asmaterial_instance_constant "IsTrunk"))
                                  (if (and (Math|Boolean|ORBoolean (== (Variables|/Script/Engine.MaterialInstance|GetParent _asmaterial_instance_constant_2) "/Game/Environments/Asteria/Foliage/Trees/SnowyPine/Materials/M_SnowyPine_master.M_SnowyPine_master") (== (Variables|/Script/Engine.MaterialInstance|GetParent _asmaterial_instance_constant_2) "/Game/Environments/Asteria/Foliage/Trees/SnowyPine/Materials/M_SnowyPine_Ice_Inst.M_SnowyPine_Ice_Inst") (== (Variables|/Script/Engine.MaterialInstance|GetParent _asmaterial_instance_constant_2) "/Game/Environments/Asteria/Foliage/Trees/SnowyPine/Materials/M_SnowyPine_Ice_NonNanite.M_SnowyPine_Ice_NonNanite")) (Math|Boolean|ORBoolean (== (Variables|/Script/Engine.MaterialInstance|GetParent _asmaterial_instance_constant) "/Game/Environments/Asteria/Foliage/Trees/SnowyPine/Materials/M_SnowyPine_master.M_SnowyPine_master") (== (Variables|/Script/Engine.MaterialInstance|GetParent _asmaterial_instance_constant) "/Game/Environments/Asteria/Foliage/Trees/SnowyPine/Materials/M_SnowyPine_Ice_Inst.M_SnowyPine_Ice_Inst") (== (Variables|/Script/Engine.MaterialInstance|GetParent _asmaterial_instance_constant) "/Game/Environments/Asteria/Foliage/Trees/SnowyPine/Materials/M_SnowyPine_Ice_NonNanite.M_SnowyPine_Ice_NonNanite")))
                                    (Development|PrintString "GeoScript NewMatIDs = SnowyPine")
                                    (Utilities|Array|Add (Variables|Default|GetnewMaterialIDs) _index_2))
                                  (if (and _returnvalue_8 _returnvalue_7)
                                    (if (== (MaterialEditing|GetMaterialInstanceStaticSwitchParameterValue _asmaterial_instance_constant_2 "IsTrunk") (MaterialEditing|GetMaterialInstanceStaticSwitchParameterValue _asmaterial_instance_constant "IsTrunk"))
                                      (Development|PrintString "GeoScript NewMatIDs = SameMaterialSwitch")))))))
                          (:then_1
                            (for (_array_element _array_index) (Variables|Default|GetnewMaterialIDs)
                              (Development|PrintString (Utilities|String|Append "GeoScript NewMatIDs =" (Utilities|String|Append (Utilities|String|Append (Utilities|String|ToString(Integer) _array_index) "  //   ") (Utilities|String|ToString(Integer) _array_element))))))
                          (:then_2
                            (for (_array_element_2 _array_index_2) (Variables|Default|GetTriangle%20IDListByMaterialID)
                              (Development|PrintString (Utilities|String|Append (Utilities|String|Append (Utilities|String|Append "GeoScript NewMatIDs = Lenght of TriangleList ==" (Utilities|String|ToString(Integer) _array_index_2)) "   Length  ") (Utilities|String|ToString(Integer) (GeometryScript|ListUtils|GetIndexListLength _array_element_2))))))
                          (:then_3
                            (bind _returnvalue_9 (DynamicMeshActor|AllocateComputeMesh))
                            (bind _returnvalue_10 (GeometryScript|Utility|CreateDynamicMeshPool))
                            (bind (_componentmeshes _componentmaterialids) (GeometryScript|Decomposition|SplitMeshbyMaterialIDs (Variables|Default|GetDynamicMesh) _returnvalue_10))
                            (if (and (== (Utilities|Array|Get(acopy) _componentmaterialids) 0) (> (Utilities|Array|Get(acopy) _componentmaterialids 1) 0))
                              (for (_ _array_index_3) _componentmaterialids
                                (bind _returnvalue_11 (GeometryScript|Materials|RemapMaterialIDs (Utilities|Array|Get(acopy) _componentmeshes (- (- (Utilities|Array|Length _componentmaterialids) 1) _array_index_3)) (- (- (Utilities|Array|Length _componentmaterialids) 1) _array_index_3) (Utilities|Array|Get(acopy) (Variables|Default|GetnewMaterialIDs) (- (- (Utilities|Array|Length _componentmaterialids) 1) _array_index_3))))
                                (GeometryScript|MeshEdits|AppendMesh _returnvalue_9 _returnvalue_11 (Math|Transform|MakeTransform "0, 0, 0") false "(CombineMode=EnableAllMatching)"))
                              (Variables|Default|SetDynamicMesh _returnvalue_9)
                              (else
                                (for (_ _array_index_4) _componentmaterialids
                                  (bind _returnvalue_12 (GeometryScript|Materials|RemapMaterialIDs (Utilities|Array|Get(acopy) _componentmeshes _array_index_4) _array_index_4 (Utilities|Array|Get(acopy) (Variables|Default|GetnewMaterialIDs) _array_index_4)))
                                  (GeometryScript|MeshEdits|AppendMesh _returnvalue_9 _returnvalue_12 (Math|Transform|MakeTransform "0, 0, 0") false "(CombineMode=EnableAllMatching)"))
                                (Variables|Default|SetDynamicMesh _returnvalue_9)))))))))
                (:NewEnumerator1
                  _)
                (:NewEnumerator2
                  _))
              (bind _returnvalue_13 (DynamicMeshActor|AllocateComputeMesh))
              (bind (_storetosubmeshout _returnvalue_14) (GeometryScript|Decomposition|GetSubMeshfromMesh _returnvalue_4 _returnvalue_13 (GeometryScript|Selection|GetTrianglesbyMaterialID _returnvalue_4 _meshmaterialid)))
              (GeometryScript|MeshEdits|DeleteTrianglesfromMesh _returnvalue_14 (GeometryScript|Selection|GetTrianglesbyMaterialID _returnvalue_4 _meshmaterialid))
              (bind _returnvalue_15 (GeometryScript|UVs|TranslateMeshUVs _storetosubmeshout 0 "(X=20.000000,Y=0.000000)"))
              (GeometryScript|MeshEdits|AppendMesh :TargetMesh _returnvalue_14 :AppendMesh _returnvalue_15 :bDeferChangeNotifications false :AppendOptions "(CombineMode=EnableAllMatching)")
              (GeometryScript|Materials|RemapMaterialIDs _returnvalue_4 1))))))
    (:then_1
      (if (Variables|Default|GetDebug)
        (BakeMesh|BakeStaticmesh)))
    (:then_2
      (if (Variables|Default|GetHasFallbackMesh)
        (bind _returnvalue_16 (DynamicMeshActor|AllocateComputeMesh))
        (bind _output_get_2 (Variables|Default|SetDynamicMesh_Fallback _returnvalue_16))
        (bind _returnvalue_17 (GeometryScript|Materials|EnableMaterialIDs _output_get_2))
        (GeometryScript|StaticMesh|CopyMeshFromStaticMeshwithSectionMaterials (Variables|Default|GetStaticmesh_Fallback) _returnvalue_17 (Utilities|Struct|MakeGeometryScriptCopyMeshfromAssetOptions true) (Utilities|Struct|MakeGeometryScriptMeshReadLOD "MaxAvailable"))))))

(event Events|EventOnRebuildGeneratedMesh
  (params (TargetMesh :type Dynamic%20Mesh%20Object%20Reference))
  (bind _staticmesh (Variables|Default|GetStaticmesh))
  (bind _staticmaterials (Variables|/Script/Engine.StaticMesh|GetStaticMaterials _staticmesh))
  (bind _returnvalue (Utilities|Array|LastIndex _staticmaterials))
  (bind _meshmaterialid (Variables|Default|GetMeshMaterialID))
  (Utilities|FlowControl|Sequence
    (:then_0
      (bind _returnvalue_2 (DynamicMeshComponent|GetDynamicMesh (Variables|DynamicMeshActor|GetDynamicMeshComponent)))
      (bind _output_get (Variables|Default|SetDynamicMesh _returnvalue_2))
      (bind _returnvalue_3 (GeometryScript|Materials|EnableMaterialIDs _output_get))
      (bind _returnvalue_4 (GeometryScript|StaticMesh|CopyMeshFromStaticMeshwithSectionMaterials (Variables|Default|GetStaticmesh) _returnvalue_3 (Utilities|Struct|MakeGeometryScriptCopyMeshfromAssetOptions true) (Utilities|Struct|MakeGeometryScriptMeshReadLOD "MaxAvailable"))
        (:Success
          (Utilities|FlowControl|Sequence
            (:then_0
              (if (not (Utilities|IsValid (Variables|Default|GetCollisionMesh) :!class /Script/Engine.KismetSystemLibrary))
                (bind _returnvalue_5 (DynamicMeshActor|AllocateComputeMesh))
                (bind _returnvalue_6 (GeometryScript|Primitives|AppendCylinder :TargetMesh _returnvalue_5 :PrimitiveOptions (Utilities|Struct|MakeGeometryScriptPrimitiveOptions "PerFace") :Radius 5e+01 :Height 5e+02 :RadialSteps 12 :HeightSteps 0 :bCapped true :Origin "Base"))
                (Variables|Default|SetHull%20Mesh _returnvalue_6)))
            (:then_1
              (switch Utilities|FlowControl|Switch|SwitchonConvertTree_Mode (Variables|Default|GetMode)
                (:NewEnumerator0
                  (if (Variables|Default|GetHasFallbackMesh)
                    (Utilities|Array|Clear (Variables|Default|GetnewMaterialIDs))
                    (Utilities|Array|Clear (Variables|Default|GetTriangle%20IDListByMaterialID))
                    (elif (Variables|Default|GetFlipMateiralIDsForNaniteandFallback)
                      _
                      (else
                        (Utilities|FlowControl|Sequence
                          (:then_0
                            (Development|PrintString (Utilities|String|Append "GeoScript NewMatIDs =" "Start"))
                            (for _index (range (+ _returnvalue 1))
                              (Utilities|Array|Add (Variables|Default|GetTriangle%20IDListByMaterialID) (GeometryScript|Selection|GetTrianglesbyMaterialID (Variables|Default|GetDynamicMesh) _index))
                              (bind _asmaterial_instance_constant (Utilities|Casting|CastToMaterialInstanceConstant (StaticMesh|GetMaterial (Variables|Default|GetStaticmesh) _index)))
                              (bind _index_2 (Utilities|FlowControl|ForLoopwithBreak 0 _returnvalue)
                                (:LoopBody
                                  (bind _asmaterial_instance_constant_2 (Utilities|Casting|CastToMaterialInstanceConstant (StaticMesh|GetMaterial (Variables|Default|GetStaticmesh_Fallback) _index_2)))
                                  (bind (_parametersource _returnvalue_7) (MaterialEditing|GetStaticSwitchParameterSource _asmaterial_instance_constant_2 "IsTrunk"))
                                  (bind (_parametersource_2 _returnvalue_8) (MaterialEditing|GetStaticSwitchParameterSource _asmaterial_instance_constant "IsTrunk"))
                                  (if (and (Math|Boolean|ORBoolean (== (Variables|/Script/Engine.MaterialInstance|GetParent _asmaterial_instance_constant_2) "/Game/Environments/Asteria/Foliage/Trees/SnowyPine/Materials/M_SnowyPine_master.M_SnowyPine_master") (== (Variables|/Script/Engine.MaterialInstance|GetParent _asmaterial_instance_constant_2) "/Game/Environments/Asteria/Foliage/Trees/SnowyPine/Materials/M_SnowyPine_Ice_Inst.M_SnowyPine_Ice_Inst") (== (Variables|/Script/Engine.MaterialInstance|GetParent _asmaterial_instance_constant_2) "/Game/Environments/Asteria/Foliage/Trees/SnowyPine/Materials/M_SnowyPine_Ice_NonNanite.M_SnowyPine_Ice_NonNanite")) (Math|Boolean|ORBoolean (== (Variables|/Script/Engine.MaterialInstance|GetParent _asmaterial_instance_constant) "/Game/Environments/Asteria/Foliage/Trees/SnowyPine/Materials/M_SnowyPine_master.M_SnowyPine_master") (== (Variables|/Script/Engine.MaterialInstance|GetParent _asmaterial_instance_constant) "/Game/Environments/Asteria/Foliage/Trees/SnowyPine/Materials/M_SnowyPine_Ice_Inst.M_SnowyPine_Ice_Inst") (== (Variables|/Script/Engine.MaterialInstance|GetParent _asmaterial_instance_constant) "/Game/Environments/Asteria/Foliage/Trees/SnowyPine/Materials/M_SnowyPine_Ice_NonNanite.M_SnowyPine_Ice_NonNanite")))
                                    (Development|PrintString "GeoScript NewMatIDs = SnowyPine")
                                    (Utilities|Array|Add (Variables|Default|GetnewMaterialIDs) _index_2))
                                  (if (and _returnvalue_8 _returnvalue_7)
                                    (if (== (MaterialEditing|GetMaterialInstanceStaticSwitchParameterValue _asmaterial_instance_constant_2 "IsTrunk") (MaterialEditing|GetMaterialInstanceStaticSwitchParameterValue _asmaterial_instance_constant "IsTrunk"))
                                      (Development|PrintString "GeoScript NewMatIDs = SameMaterialSwitch")))))))
                          (:then_1
                            (for (_array_element _array_index) (Variables|Default|GetnewMaterialIDs)
                              (Development|PrintString (Utilities|String|Append "GeoScript NewMatIDs =" (Utilities|String|Append (Utilities|String|Append (Utilities|String|ToString(Integer) _array_index) "  //   ") (Utilities|String|ToString(Integer) _array_element))))))
                          (:then_2
                            (for (_array_element_2 _array_index_2) (Variables|Default|GetTriangle%20IDListByMaterialID)
                              (Development|PrintString (Utilities|String|Append (Utilities|String|Append (Utilities|String|Append "GeoScript NewMatIDs = Lenght of TriangleList ==" (Utilities|String|ToString(Integer) _array_index_2)) "   Length  ") (Utilities|String|ToString(Integer) (GeometryScript|ListUtils|GetIndexListLength _array_element_2))))))
                          (:then_3
                            (bind _returnvalue_9 (DynamicMeshActor|AllocateComputeMesh))
                            (bind _returnvalue_10 (GeometryScript|Utility|CreateDynamicMeshPool))
                            (bind (_componentmeshes _componentmaterialids) (GeometryScript|Decomposition|SplitMeshbyMaterialIDs (Variables|Default|GetDynamicMesh) _returnvalue_10))
                            (if (and (== (Utilities|Array|Get(acopy) _componentmaterialids) 0) (> (Utilities|Array|Get(acopy) _componentmaterialids 1) 0))
                              (for (_ _array_index_3) _componentmaterialids
                                (bind _returnvalue_11 (GeometryScript|Materials|RemapMaterialIDs (Utilities|Array|Get(acopy) _componentmeshes (- (- (Utilities|Array|Length _componentmaterialids) 1) _array_index_3)) (- (- (Utilities|Array|Length _componentmaterialids) 1) _array_index_3) (Utilities|Array|Get(acopy) (Variables|Default|GetnewMaterialIDs) (- (- (Utilities|Array|Length _componentmaterialids) 1) _array_index_3))))
                                (GeometryScript|MeshEdits|AppendMesh _returnvalue_9 _returnvalue_11 (Math|Transform|MakeTransform "0, 0, 0") false "(CombineMode=EnableAllMatching)"))
                              (Variables|Default|SetDynamicMesh _returnvalue_9)
                              (else
                                (for (_ _array_index_4) _componentmaterialids
                                  (bind _returnvalue_12 (GeometryScript|Materials|RemapMaterialIDs (Utilities|Array|Get(acopy) _componentmeshes _array_index_4) _array_index_4 (Utilities|Array|Get(acopy) (Variables|Default|GetnewMaterialIDs) _array_index_4)))
                                  (GeometryScript|MeshEdits|AppendMesh _returnvalue_9 _returnvalue_12 (Math|Transform|MakeTransform "0, 0, 0") false "(CombineMode=EnableAllMatching)"))
                                (Variables|Default|SetDynamicMesh _returnvalue_9)))))))))
                (:NewEnumerator1
                  _)
                (:NewEnumerator2
                  _))
              (bind _returnvalue_13 (DynamicMeshActor|AllocateComputeMesh))
              (bind (_storetosubmeshout _returnvalue_14) (GeometryScript|Decomposition|GetSubMeshfromMesh _returnvalue_4 _returnvalue_13 (GeometryScript|Selection|GetTrianglesbyMaterialID _returnvalue_4 _meshmaterialid)))
              (GeometryScript|MeshEdits|DeleteTrianglesfromMesh _returnvalue_14 (GeometryScript|Selection|GetTrianglesbyMaterialID _returnvalue_4 _meshmaterialid))
              (bind _returnvalue_15 (GeometryScript|UVs|TranslateMeshUVs _storetosubmeshout 0 "(X=20.000000,Y=0.000000)"))
              (GeometryScript|MeshEdits|AppendMesh :TargetMesh _returnvalue_14 :AppendMesh _returnvalue_15 :bDeferChangeNotifications false :AppendOptions "(CombineMode=EnableAllMatching)")
              (GeometryScript|Materials|RemapMaterialIDs _returnvalue_4 1))))))
    (:then_1
      (if (Variables|Default|GetDebug)
        (BakeMesh|BakeStaticmesh)))
    (:then_2
      (if (Variables|Default|GetHasFallbackMesh)
        (bind _returnvalue_16 (DynamicMeshActor|AllocateComputeMesh))
        (bind _output_get_2 (Variables|Default|SetDynamicMesh_Fallback _returnvalue_16))
        (bind _returnvalue_17 (GeometryScript|Materials|EnableMaterialIDs _output_get_2))
        (GeometryScript|StaticMesh|CopyMeshFromStaticMeshwithSectionMaterials (Variables|Default|GetStaticmesh_Fallback) _returnvalue_17 (Utilities|Struct|MakeGeometryScriptCopyMeshfromAssetOptions true) (Utilities|Struct|MakeGeometryScriptMeshReadLOD "MaxAvailable"))))))

(fn BakeStaticmesh
  (bind _meshnanitesettings (Utilities|Struct|MakeMeshNaniteSettings true))
  (bind _asset_path_and_name (Variables|Default|GetAsset%20Path%20and%20Name))
  (Development|PrintString (Utilities|String|Append "GeoScript :  " (Utilities|GetDisplayName (Variables|Default|GetStaticmesh))))
  (bind _returnvalue (EditorScripting|Asset|DoesAssetExist _asset_path_and_name :!class /Script/EditorScriptingUtilities.EditorAssetLibrary))
  (if _returnvalue
    (bind _returnvalue_2 (EditorScripting|Asset|LoadAsset _asset_path_and_name :!class /Script/EditorScriptingUtilities.EditorAssetLibrary))
    (bind _asstatic_mesh (Utilities|Casting|CastToStaticMesh _returnvalue_2))
    (Utilities|FlowControl|Sequence
      (:then_0
        (if (Variables|Default|GetHasFallbackMesh)
          (GeometryScript|StaticMesh|CopyMeshtoStaticMesh (Variables|Default|GetDynamicMesh_Fallback) _asstatic_mesh (Utilities|Struct|MakeGeometryScriptCopyMeshtoAssetOptions :bEnableRecomputeNormals false :bEnableRecomputeTangents false :bEnableRemoveDegenerates false :BoneHierarchyMismatchHandling "DoNothing" :bRemapBoneIndicesToMatchAsset false :bUseOriginalVertexOrder false :bUseBuildScale true :bReplaceMaterials false :NewMaterialUsage "ReplaceAssetMaterials" :bCleanAssignedMaterials false :GenerateLightmapUVs "MatchTargetLODSetting" :bApplyNaniteSettings true :NaniteSettings "(bEnabled=True,FallbackPercentTriangles=100.000000,FallbackRelativeError=0.000000)" :NewNaniteSettings (Utilities|Struct|MakeMeshNaniteSettings true) :bEmitTransaction true :bDeferMeshPostEditChange false) (Utilities|Struct|MakeGeometryScriptMeshWriteLOD false))))
      (:then_1
        (Utilities|IsValid (Variables|Default|GetStaticmesh)
          (:Is%20Valid
            (GeometryScript|StaticMesh|CopyMeshtoStaticMesh (Variables|Default|GetDynamicMesh) _asstatic_mesh (Utilities|Struct|MakeGeometryScriptCopyMeshtoAssetOptions :bEnableRecomputeNormals false :bEnableRecomputeTangents false :bEnableRemoveDegenerates false :BoneHierarchyMismatchHandling "DoNothing" :bRemapBoneIndicesToMatchAsset false :bUseOriginalVertexOrder false :bUseBuildScale true :bReplaceMaterials false :NewMaterialUsage "ReplaceAssetMaterials" :bCleanAssignedMaterials false :GenerateLightmapUVs "MatchTargetLODSetting" :bApplyNaniteSettings true :NaniteSettings "(bEnabled=True,FallbackPercentTriangles=100.000000,FallbackRelativeError=0.000000)" :NewNaniteSettings _meshnanitesettings :bEmitTransaction true :bDeferMeshPostEditChange false) (Utilities|Struct|MakeGeometryScriptMeshWriteLOD true)
              (:Success
                (CallFunction|SetCollision _asstatic_mesh)
                (return _asstatic_mesh)))))))
    (else
      (bind _returnvalue_3 (GeometryScript|AssetManagement|CreateNewStaticMeshAssetfromMesh (Variables|Default|GetDynamicMesh) (Variables|Default|GetAsset%20Path%20and%20Name) (Utilities|Struct|MakeGeometryScriptCreateNewStaticMeshAssetOptions false false true _meshnanitesettings))
        (:Success
          (CallFunction|SetCollision _returnvalue_3)
          (return _returnvalue_3))))))

(fn InfuseCustomLOD
  (bind _asset_path_and_name (Variables|Default|GetAsset%20Path%20and%20Name))
  (bind _returnvalue (DynamicMeshActor|AllocateComputeMesh))
  (bind _returnvalue_2 (GeometryScript|StaticMesh|CopyMeshFromStaticMeshwithSectionMaterials (Variables|Default|GetCustomLODMesh) _returnvalue (Utilities|Struct|MakeGeometryScriptCopyMeshfromAssetOptions true) (Utilities|Struct|MakeGeometryScriptMeshReadLOD "MaxAvailable"))
    (:Success
      (bind _returnvalue_3 (EditorScripting|Asset|DoesAssetExist _asset_path_and_name :!class /Script/EditorScriptingUtilities.EditorAssetLibrary))
      (if _returnvalue_3
        (bind _returnvalue_4 (EditorScripting|Asset|LoadAsset _asset_path_and_name :!class /Script/EditorScriptingUtilities.EditorAssetLibrary))
        (bind _asstatic_mesh (Utilities|Casting|CastToStaticMesh _returnvalue_4))
        (GeometryScript|StaticMesh|CopyMeshtoStaticMesh _returnvalue_2 _asstatic_mesh (Utilities|Struct|MakeGeometryScriptCopyMeshtoAssetOptions :bEnableRecomputeNormals false :bEnableRecomputeTangents false :bEnableRemoveDegenerates false :BoneHierarchyMismatchHandling "DoNothing" :bRemapBoneIndicesToMatchAsset false :bUseOriginalVertexOrder false :bUseBuildScale true :bReplaceMaterials false :NewMaterialUsage "ReplaceAssetMaterials" :bCleanAssignedMaterials false :GenerateLightmapUVs "MatchTargetLODSetting" :bApplyNaniteSettings true :NaniteSettings "(bEnabled=True,FallbackPercentTriangles=100.000000,FallbackRelativeError=0.000000)" :NewNaniteSettings (Utilities|Struct|MakeMeshNaniteSettings true) :bEmitTransaction true :bDeferMeshPostEditChange false) (Utilities|Struct|MakeGeometryScriptMeshWriteLOD false (Variables|Default|GetTargetLODindex)))))))

(fn SetCollision
  (params (InputPin :type Static%20Mesh%20Object%20Reference))
  (bind _geometryscriptcollisionfrommeshoptions (Utilities|Struct|MakeGeometryScriptCollisionfromMeshOptions true "ConvexHulls" false false false 1.0 false))
  (if (Utilities|IsValid (Variables|Default|GetCollisionMesh) :!class /Script/Engine.KismetSystemLibrary)
    (bind _returnvalue (DynamicMeshActor|AllocateComputeMesh))
    (bind _returnvalue_2 (GeometryScript|StaticMesh|CopyMeshFromStaticMeshwithSectionMaterials (Variables|Default|GetCollisionMesh) _returnvalue)
      (:Success
        (GeometryScript|Collision|SetStaticMeshCollisionfromMesh _returnvalue_2 InputPin _geometryscriptcollisionfrommeshoptions "(bMarkAsCustomized=True)")))
    (else
      (GeometryScript|Collision|SetStaticMeshCollisionfromMesh (Variables|Default|GetHull%20Mesh) InputPin _geometryscriptcollisionfrommeshoptions "(bMarkAsCustomized=True)"))))

(fn ConstructionScript)
