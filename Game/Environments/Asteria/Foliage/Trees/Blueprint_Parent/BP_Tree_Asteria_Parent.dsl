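; Parent Blueprint for Asteria trees that sets up leaf color, texture variants, a shadow proxy and optional falling leaves.
; Construction applies seasonal presets from a data asset or custom colors, writing them and texture array indices to custom primitive data.
; BeginPlay can tighten WPO disable distance, gated by material quality and FN.Tree.WPOQualityLevel.
; The shadow proxy's distance check writes to the main mesh, not the proxy.
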(event EventBeginPlay
  (CallFunction|BeginPlayWPOQualityOverrides))

(fn BeginPlayWPOQualityOverrides
  (if (== (Rendering|Material|GetRenderingMaterialQualityLevel) 3)
    (bind _returnvalue (Development|GetConsoleVariableIntValue "FN.Tree.WPOQualityLevel"))
    (if (and (== _returnvalue 0) (!= (Variables|Default|GetLowQualityWPODisableDistance) 0))
      (Utilities|FlowControl|Sequence
        (:then_0
          (if (or (< (Variables|Default|GetLowQualityWPODisableDistance) (Variables|/Script/Engine.StaticMeshComponent|GetWorldPositionOffsetDisableDistance (Variables|Debug|GetStaticMeshComponent))) (== (Variables|/Script/Engine.StaticMeshComponent|GetWorldPositionOffsetDisableDistance (Variables|Debug|GetStaticMeshComponent)) 0))
            (Rendering|LOD|SetWorldPositionOffsetDisableDistance (Variables|Debug|GetStaticMeshComponent) (Variables|Default|GetLowQualityWPODisableDistance) :!class /Script/Engine.StaticMeshComponent)))
        (:then_1
          (if (or (< (Variables|Default|GetLowQualityWPODisableDistance) (Variables|/Script/Engine.StaticMeshComponent|GetWorldPositionOffsetDisableDistance (Variables|Default|GetShadowProxyMeshComponent))) (== (Variables|/Script/Engine.StaticMeshComponent|GetWorldPositionOffsetDisableDistance (Variables|Default|GetShadowProxyMeshComponent)) 0))
            (Rendering|LOD|SetWorldPositionOffsetDisableDistance (Variables|Debug|GetStaticMeshComponent) (Variables|Default|GetLowQualityWPODisableDistance) :!class /Script/Engine.StaticMeshComponent)))))))

(fn SetArrayTextureCustomPrimData
  (Rendering|Material|SetCustomPrimitiveDataFloat (Variables|Debug|GetStaticMeshComponent) 10 (Math|Conversions|ToFloat(Integer) (Variables|Default|GetArrayIndex)))
  (Rendering|Material|SetCustomPrimitiveDataFloat (Variables|Debug|GetStaticMeshComponent) 31 (Math|Conversions|ToFloat(Integer) (Variables|Default|GetArrayIndex_Trunk)))
  (Rendering|Material|SetCustomPrimitiveDataFloat (Variables|Debug|GetStaticMeshComponent) 30 (Math|Conversions|ToFloat(Integer) (Variables|Default|GetArrayIndex_Leaves))))

(fn SetSeasonPresetColors
  (bind _array_element (Utilities|Array|ForEachLoopwithBreak (Variables|/Game/ArtTools/Procedural/Trees/TreeSeasonColor_Data.TreeSeasonColor_Data_C|GetTreeSeasonData (Variables|LeaveColor|GetTreeSeasonColor%20Data)))
    (:LoopBody
      (if (Utilities|Enum|Equal(Enum) (Utilities|Struct|BreakTreeSeasonColorStruct _array_element) (Variables|LeaveColor|GetTree%20Color%20Option))
        (Variables|LeaveColor|SetColor (.Color_5_E33006414CF6D999D4D482A4D2FF03F6 (Utilities|Struct|BreakTreeSeasonColorStruct _array_element)))
        (Variables|LeaveColor|SetHue_Shift_BlueVertexColor (.ValueVariation_9_4FAED9E04DA8955B0BC4EDB1E0A3C1F7 (Utilities|Struct|BreakTreeSeasonColorStruct _array_element)))
        (Variables|LeaveColor|SetHueVariation_PerLeaf (.HueVariation_10_00BFFB9140C4AD7D16C9AA9AEB01A935 (Utilities|Struct|BreakTreeSeasonColorStruct _array_element)))))))

(fn SetTreeSeasonalColor
  (Utilities|FlowControl|Sequence
    (:then_0
      (if (Utilities|Enum|NotEqual(Enum) (Variables|LeaveColor|GetTree%20Color%20Option) "NewEnumerator0")
        (Default|SetSeasonPresetColors)))
    (:then_1
      (Default|UpdateColor))))

(fn ShadowProxyLogic
  (Class|ShadowProxyMeshComponent|EditorConstruct (Variables|Default|GetShadowProxyMeshComponent)))

(fn UpdateColor
  (bind (_r _g _b _a) (Utilities|Struct|BreakLinearColor (Variables|LeaveColor|GetColor)))
  (bind (_r_2 _g_2 _b_2) (Utilities|Struct|BreakLinearColor (Variables|LeaveColor|GetColor_2)))
  (bind _staticmeshcomponent (Variables|Debug|GetStaticMeshComponent))
  (Rendering|Material|SetDefaultCustomPrimitiveDataVector4 _staticmeshcomponent 0 (Math|Vector4|MakeVector4 _r _g _b _a))
  (Rendering|Material|SetDefaultCustomPrimitiveDataVector3 _staticmeshcomponent 17 (Math|Vector|MakeVector _r_2 _g_2 _b_2))
  (Rendering|Material|SetDefaultCustomPrimitiveDataFloat (Variables|Debug|GetStaticMeshComponent) 4 (Variables|LeaveColor|GetHueVariation_PerLeaf))
  (Rendering|Material|SetDefaultCustomPrimitiveDataFloat (Variables|Debug|GetStaticMeshComponent) 5 (Variables|LeaveColor|GetHue_Shift_BlueVertexColor))
  (Rendering|Material|SetDefaultCustomPrimitiveDataFloat (Variables|Debug|GetStaticMeshComponent) 6 (Utilities|Select :Index (Variables|LeaveColor|GetTree%20Color%20Option) :NewEnumerator0 0.0 :NewEnumerator1 0.0 :NewEnumerator2 0.0 :NewEnumerator3 0.0 :NewEnumerator4 1.0)))

(fn ConstructionScript
  (Utilities|FlowControl|Sequence
    (:then_0
      (Default|SetTreeSeasonalColor)
      (CallFunction|ShadowProxyLogic)
      (CallFunction|SetArrayTextureCustomPrimData))
    (:then_1
      ;[1 1 1] Falling Leaves
      (if (Variables|FX|GetHas%20Falling%20Leaves)
        (AddComponent|NODE_AddStaticMeshComponent-0 :RelativeTransform (Math|Conversions|ToTransform(Vector) (Variables|FX|GetLeavesOffset)))))))
