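; Parent Blueprint for Asteria trees that switches holiday lights on or off per tree.
; EnableHolidayLights writes custom primitive data float 31 on the static mesh (0 enabled, 1 disabled), using the default-data setter during construction.
; BeginPlay re-sets the RepNotify EnableLights to itself so OnRep reapplies the value whenever the tree streams in.
; Per its comments, an external trigger such as a calendar event should drive it, and impostors are unsupported.
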(event EventBeginPlay
  (|CallParentFunction)
  ; Trigger this from you Calendar Event or what ever triggers the Lights on the tree.
  ; It was requested to be not global but per tree.
  ; This is Primitive Data set per tree this need to update every time the tree is streamed in.
  ; Only works on the tree it self. Doesn't work with Impostor
  (Variables|Default|SetEnableLights (Variables|Default|GetEnableLights)))

(event Collision|EventActorBeginOverlap
  (params (OtherActor :type Actor%20Object%20Reference)))

(event EventTick
  (params (DeltaSeconds :type Float%20%28single-precision%29)))

(fn EnableHolidayLights
  (params (Enable :type Boolean) (isConstructionScript :type Boolean))
  (bind _returnvalue (Math|Float|SelectFloat 0.0 1.0 Enable))
  (bind _staticmeshcomponent (Variables|Debug|GetStaticMeshComponent))
  (if isConstructionScript
    (Rendering|Material|SetDefaultCustomPrimitiveDataFloat _staticmeshcomponent 31 _returnvalue) ; 0 = Enabled and 1 = Disabled
    (else
      (Rendering|Material|SetCustomPrimitiveDataFloat _staticmeshcomponent 31 _returnvalue))))

(fn OnRep_EnableLights
  (CallFunction|EnableHolidayLights (Variables|Default|GetEnableLights)))

(fn ConstructionScript
  (|CallParentFunction)
  (CallFunction|EnableHolidayLights (Variables|Default|GetEnableLights) true))
