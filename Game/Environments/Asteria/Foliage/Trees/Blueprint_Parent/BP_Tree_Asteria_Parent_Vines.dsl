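; Parent Blueprint for Asteria trees that procedurally dresses them with vines and small plants, using seeded random streams.
; Construction links "Vine" sockets with instanced hanging vines and optionally adds a random addon mesh and trace-placed plants.
; In editor, outward cone traces from those sockets find surfaces for support vines and connectors, saving transforms in arrays.
; These rebuild in edit mode or when the actor's transform changes, then feed two instanced mesh components.
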(event EventBeginPlay)

(event Collision|EventActorBeginOverlap
  (params (OtherActor :type Actor%20Object%20Reference)))

(event EventTick
  (params (DeltaSeconds :type Float%20%28single-precision%29)))

(event Custom|Spawn%20Support%20Vines
  (Development|Editor|CreateCopyforUndoBuffer self)
  (Default|SpawnSupportVines false)
  (Default|SpawnVineInstances))

(fn BuildExposedAssetProviderVersion(Implementation)
  (params (Version :type Fort%20Exposed%20Asset%20Provider%20Version%20Structure%20%28by%20ref%29))
  (|CallParentFunction Version)
  (ExposedAssetProvider|AddVersion Version 1)
  (return))

(fn CollectAssetReferencestoAllow(Implementation)
  (params (Collector :type Fort%20Exposed%20Asset%20Reference%20Collector%20Structure%20%28by%20ref%29))
  (|CallParentFunction Collector)
  (ExposedAssetProvider|CollectObjectPropertyAssetReferences Collector 0 (Variables|Default|GetAddonStaticmeshOption1))
  (ExposedAssetProvider|CollectObjectPropertyAssetReferences Collector 0 (Variables|Default|GetAddonStaticmeshOption2))
  (return))

(fn ConnectToSocket
  (params (Actor :type Actor%20Object%20Reference) (Stream :type Random%20Stream%20Structure%20%28by%20ref%29))
  (Utilities|FlowControl|Sequence
    (:then_0
      (for _array_element (Actor|GetComponentsByClass Actor "/Script/Engine.StaticMeshComponent")
        (for _array_element_2 (Transformation|GetAllSocketNames _array_element)
          (if (Utilities|String|Contains (Utilities|String|ToString(Name) _array_element_2) "Vine")
            (Utilities|Array|Add (LocalVariables|GetLocationList) (.location (Transformation|GetSocketTransform _array_element _array_element_2 "RTS_World")))))))
    (:then_1
      (return (Utilities|Array|Get(acopy) (LocalVariables|GetLocationList) (Math|Random|RandomIntegerinRangefromStream Stream 0 (Utilities|Array|LastIndex (LocalVariables|GetLocationList)))) (>= (Utilities|Array|LastIndex (LocalVariables|GetLocationList)) 0)))))

(fn SpawnSmall_Plants
  (bind _stream (Variables|Default|GetStream))
  (if (Variables|Default|GetSpawnSmallPlants)
    (Math|Random|SetRandomStreamSeed (Variables|Default|GetStream) (Math|Conversions|ToInteger(Integer64) (+ (Variables|Default|GetSeed) (Math|Float|TruncatetoInteger64 (+ (.x (Transformation|GetActorLocation)) (.y (Transformation|GetActorLocation)))))))
    (bind _returnvalue (AddComponent|NODE_AddHierarchicalInstancedStaticMeshComponent-0 :RelativeTransform (Math|Transform|MakeTransform "0, 0, 0")))
    (LocalVariables|SetISM _returnvalue)
    (for _ (range (+ (Math|Random|RandomIntegerinRangefromStream _stream 10 (LocalVariables|GetMaxTraces)) 1))
      (bind (_outhit _returnvalue_2) (Collision|LineTraceByChannel :Start (Utilities|Operators|Add (* (Math|Vector|RotateVectorAroundAxis "1.000000,0.000000,0.000000" (Math|Random|RandomFloatinRangefromStream _stream 0.0 3.6e+02) "0.000000,0.000000,1.000000") 2e+02) (Transformation|GetActorLocation) (Math|Vector|MakeVector 0.0 0.0 (Math|Random|RandomFloatinRangefromStream (Variables|Default|GetStream) 5.0 3e+02))) :End (+ (Utilities|Operators|Add (* (Math|Vector|RotateVectorAroundAxis "1.000000,0.000000,0.000000" (Math|Random|RandomFloatinRangefromStream _stream 0.0 3.6e+02) "0.000000,0.000000,1.000000") 2e+02) (Transformation|GetActorLocation) (Math|Vector|MakeVector 0.0 0.0 (Math|Random|RandomFloatinRangefromStream (Variables|Default|GetStream) 5.0 3e+02))) (* (Math|Random|RandomUnitVectorinConeinDegreesfromStream (Variables|Default|GetStream) (Math|Vector|NegateVector (Math|Vector|RotateVectorAroundAxis "1.000000,0.000000,0.000000" (Math|Random|RandomFloatinRangefromStream _stream 0.0 3.6e+02) "0.000000,0.000000,1.000000")) 6e+01) 2e+02)) :TraceChannel "TraceTypeQuery10" :bTraceComplex true :DrawDebugType "None" :bIgnoreSelf false :TraceColor "(R=1.000000,G=0.000000,B=0.000000,A=1.000000)" :TraceHitColor "(R=0.000000,G=1.000000,B=0.000000,A=1.000000)" :DrawTime 1.0))
      (if _returnvalue_2
        (Utilities|FlowControl|Sequence
          (:then_0
            (LocalVariables|SetValidHit (== (.HitActor (Collision|BreakHitResult _outhit)) self)))
          (:then_1
            (Utilities|Casting|CastToLandscapeProxy (.HitActor (Collision|BreakHitResult _outhit)))
            (LocalVariables|SetValidHit true))
          (:then_2
            (if (and (LocalVariables|GetValidHit) _returnvalue_2)
              (Components|InstancedStaticMesh|AddInstance (LocalVariables|GetISM) (Math|Transform|MakeTransform (.Location (Collision|BreakHitResult _outhit)) (Math|Rotator|MakeRotfromZ (.Normal (Collision|BreakHitResult _outhit)))) true))))))))

(fn SpawnSupportVines
  (params (IsEnd :type Boolean))
  (bind _staticmeshcomponent (Variables|Debug|GetStaticMeshComponent))
  (bind _supportvineslist (Variables|Default|GetSupportVinesList))
  (bind _overridehitlocation (LocalVariables|GetOverrideHitLocation))
  (bind _overridehit_location (LocalVariables|GetOverrideHit_Location))
  (Variables|Default|SetEdit%20Mode true)
  (Utilities|Array|Clear (Variables|Default|GetSupportVines_ConnectorsList))
  (Utilities|Array|Clear (Variables|Default|GetSupportVinesList))
  (Components|InstancedStaticMesh|ClearInstances (Variables|Default|GetConnectionsVines_HISM))
  (Components|InstancedStaticMesh|ClearInstances (Variables|Default|GetConnectionsVines_Connector_HISM))
  (Math|Random|SetRandomStreamSeed (Variables|Default|GetStream) (Variables|Default|GetSeed))
  (LocalVariables|SetCount 0)
  (Utilities|FlowControl|Sequence
    (:then_0
      (Utilities|FlowControl|Sequence
        (:then_0
          (for _array_element (Transformation|GetAllSocketNames _staticmeshcomponent)
            (if (Utilities|String|Contains (Utilities|String|ToString(Name) _array_element) "Vine")
              (Utilities|Array|Add (LocalVariables|GetVineLocations) (Transformation|GetSocketTransform _staticmeshcomponent _array_element "RTS_World")))))))
    (:then_1
      (bind _array_element_2 (Utilities|Array|ForEachLoopwithBreak (LocalVariables|GetVineLocations))
        (:LoopBody
          (Utilities|FlowControl|ForLoopwithBreak 0 (LocalVariables|GetMaxTracesPerPoint)
            (:LoopBody
              (LocalVariables|SetisvalidHit false)
              (bind _output_get (LocalVariables|SetStartPoint (.location _array_element_2)))
              (bind (_outhit _returnvalue) (Collision|LineTraceByChannel :Start _output_get :End (+ (.location _array_element_2) (* (Math|Random|RandomUnitVectorinConeinDegreesfromStream (Variables|Default|GetStream) (Math|Vector|Normalize (+ (Math|Vector|Normalize (* (- (.location _array_element_2) (Transformation|GetActorLocation)) "1.000000,1.000000,0.000000")) "0.000000,0.000000,0.050000")) 45.0) 2e+03)) :TraceChannel "TraceTypeQuery10" :bTraceComplex true :DrawDebugType "None" :bIgnoreSelf true :TraceColor "(R=1.000000,G=0.000000,B=0.000000,A=1.000000)" :TraceHitColor "(R=0.000000,G=1.000000,B=0.000000,A=1.000000)" :DrawTime 1.0))
              (LocalVariables|SetOverrideHitLocation false)
              (LocalVariables|SetSpawnConnector true)
              (Utilities|FlowControl|Sequence
                (:then_0
                  _)
                (:then_1
                  ;[1 1 1] Check for valid actors that are not destructible
                  (bind _asbuilding_actor (Utilities|Casting|CastToBuildingActor (.HitActor (Collision|BreakHitResult _outhit))))
                  (LocalVariables|SetisvalidHit (not (Variables|/Script/Engine.Actor|GetbCanBeDamaged _asbuilding_actor))))
                (:then_2
                  (Utilities|Casting|CastToLandscapeProxy (.HitActor (Collision|BreakHitResult _outhit)))
                  (if (not (> (Math|Vector|DotProduct "0, 0, 0" "0.000000,0.000000,1.000000") (Math|Random|RandomFloatinRangefromStream (Variables|Default|GetStream) 0.5 1.0)))
                    (LocalVariables|SetisvalidHit true)))
                (:then_3
                  (Utilities|FlowControl|Sequence))
                (:then_5
                  (if (and _returnvalue (LocalVariables|GetisvalidHit))
                    (Math|Integer|IncrementInt (LocalVariables|GetCount)
                      (:%20%20
                        (bind _returnvalue_2 (Utilities|Array|Add _supportvineslist (Math|Transform|MakeTransform (LocalVariables|GetStartPoint) (Math|Rotator|FindLookatRotation (LocalVariables|GetStartPoint) (select _overridehitlocation _overridehit_location (.Location (Collision|BreakHitResult _outhit)))) (Math|Vector|MakeVector (/ (Math|Vector|Distance(Vector) (LocalVariables|GetStartPoint) (select _overridehitlocation _overridehit_location (.Location (Collision|BreakHitResult _outhit)))) 5e+02) (/ (Math|Vector|Distance(Vector) (LocalVariables|GetStartPoint) (select _overridehitlocation _overridehit_location (.Location (Collision|BreakHitResult _outhit)))) 5e+02) (Math|Random|RandomFloatinRangefromStream (Variables|Default|GetStream) 5e+01 1e+02)))))
                        (if (LocalVariables|GetSpawnConnector)
                          (Utilities|Array|Add (Variables|Default|GetSupportVines_ConnectorsList) (Math|Transform|MakeTransform (+ (select _overridehitlocation _overridehit_location (.Location (Collision|BreakHitResult _outhit))) (+ (* (Math|Vector|GetUnitDirection(Vector) (select _overridehitlocation _overridehit_location (.Location (Collision|BreakHitResult _outhit))) (LocalVariables|GetStartPoint)) (* 65.0 (.x (.scale (Utilities|Array|Get(acopy) _supportvineslist _returnvalue_2))))) (Utilities|Operators|Multiply (Math|Vector|NegateVector (Math|Vector|GetUpVector (.rotation (Utilities|Array|Get(acopy) _supportvineslist _returnvalue_2)))) (.z (.scale (Utilities|Array|Get(acopy) _supportvineslist _returnvalue_2))) 0.41))) (Math|Rotator|CombineRotators (Math|Rotator|MakeRotfromX (Math|Vector|GetUnitDirection(Vector) (+ (select _overridehitlocation _overridehit_location (.Location (Collision|BreakHitResult _outhit))) (+ (* (Math|Vector|GetUnitDirection(Vector) (select _overridehitlocation _overridehit_location (.Location (Collision|BreakHitResult _outhit))) (LocalVariables|GetStartPoint)) (* 65.0 (.x (.scale (Utilities|Array|Get(acopy) _supportvineslist _returnvalue_2))))) (Utilities|Operators|Multiply (Math|Vector|NegateVector (Math|Vector|GetUpVector (.rotation (Utilities|Array|Get(acopy) _supportvineslist _returnvalue_2)))) (.z (.scale (Utilities|Array|Get(acopy) _supportvineslist _returnvalue_2))) 0.41))) (select _overridehitlocation _overridehit_location (.Location (Collision|BreakHitResult _outhit))))) (Math|Vector|RotatorfromAxisandAngle (Math|Vector|GetForwardVector (Math|Rotator|MakeRotfromX (Math|Vector|GetUnitDirection(Vector) (+ (select _overridehitlocation _overridehit_location (.Location (Collision|BreakHitResult _outhit))) (+ (* (Math|Vector|GetUnitDirection(Vector) (select _overridehitlocation _overridehit_location (.Location (Collision|BreakHitResult _outhit))) (LocalVariables|GetStartPoint)) (* 65.0 (.x (.scale (Utilities|Array|Get(acopy) _supportvineslist _returnvalue_2))))) (Utilities|Operators|Multiply (Math|Vector|NegateVector (Math|Vector|GetUpVector (.rotation (Utilities|Array|Get(acopy) _supportvineslist _returnvalue_2)))) (.z (.scale (Utilities|Array|Get(acopy) _supportvineslist _returnvalue_2))) 0.41))) (select _overridehitlocation _overridehit_location (.Location (Collision|BreakHitResult _outhit)))))) (Math|Random|RandomFloatinRangefromStream (Variables|Default|GetStream) 0.0 3.6e+02))) (Math|Vector|MakeVector (.x (.scale (Utilities|Array|Get(acopy) _supportvineslist _returnvalue_2))) 1.0 1.0))))
                        (Utilities|FlowControl|Sequence
                          (:then_0
                            (if (>= (LocalVariables|GetCount) (LocalVariables|GetMaxVinesPerPoint))
                              _))))))))
              (Utilities|Casting|CastToStaticMeshActor (.HitActor (Collision|BreakHitResult _outhit)))
              (LocalVariables|SetisvalidHit true))))))))

(fn SpawnVineInstances
  (Utilities|FlowControl|Sequence
    (:then_0
      (for _array_element (Variables|Default|GetSupportVinesList)
        (Components|InstancedStaticMesh|AddInstance (Variables|Default|GetConnectionsVines_HISM) _array_element true)))
    (:then_1
      (for _array_element_2 (Variables|Default|GetSupportVines_ConnectorsList)
        (Components|InstancedStaticMesh|AddInstance (Variables|Default|GetConnectionsVines_Connector_HISM) _array_element_2 true)))))

(fn SpawnVines
  (bind _staticmeshcomponent (Variables|Debug|GetStaticMeshComponent))
  (bind _instancevines_tree_hanging (Variables|Default|GetInstanceVines%20Tree%20Hanging))
  (bind _vinelocations (LocalVariables|GetVineLocations))
  (bind _vinelocations_2 (LocalVariables|GetVineLocations))
  (bind _blacklist (LocalVariables|Getblacklist))
  (if (Variables|Default|GetSpawnHangingVines)
    (Utilities|FlowControl|Sequence
      (:then_0
        (for _array_element (Transformation|GetAllSocketNames _staticmeshcomponent)
          (if (Utilities|String|Contains (Utilities|String|ToString(Name) _array_element) "Vine")
            (Utilities|Array|Add (LocalVariables|GetVineLocations) (Transformation|GetSocketTransform _staticmeshcomponent _array_element "RTS_World")))))
      (:then_1
        (LocalVariables|SetVineLocations_Temp _vinelocations_2)
        (bind (_array_element_2 _array_index) (Utilities|Array|ForEachLoopwithBreak _vinelocations_2)
          (:LoopBody
            (Utilities|Array|Add _blacklist _array_index)
            (LocalVariables|SetConnectionCount (LocalVariables|GetmaxConnections_0))
            (bind (_array_element_3 _array_index_2) (Utilities|Array|ForEachLoopwithBreak _vinelocations)
              (:LoopBody
                (if (> (LocalVariables|GetConnectionCount) 0)
                  (if (not (>= (Utilities|Array|FindItem _blacklist _array_index_2) 0))
                    (Math|Integer|DecrementInt (LocalVariables|GetConnectionCount)
                      (:%20%20
                        (bind _returnvalue (Components|InstancedStaticMesh|AddInstance _instancevines_tree_hanging (Math|Transform|MakeTransform (+ (.location _array_element_2) (* (Math|Vector|GetUnitDirection(Vector) (.location _array_element_2) (.location (Utilities|Array|Get(acopy) _vinelocations _array_index_2))) (/ (Math|Vector|Distance(Vector) (.location (Utilities|Array|Get(acopy) _vinelocations _array_index_2)) (.location _array_element_2)) 2.0))) (Math|Rotator|MakeRotfromX (Math|Vector|GetUnitDirection(Vector) (.location _array_element_2) (.location (Utilities|Array|Get(acopy) _vinelocations _array_index_2)))) (Math|Vector|MakeVector (/ (Math|Vector|Distance(Vector) (.location (Utilities|Array|Get(acopy) _vinelocations _array_index_2)) (.location _array_element_2)) 378.0) (/ (Math|Vector|Distance(Vector) (.location (Utilities|Array|Get(acopy) _vinelocations _array_index_2)) (.location _array_element_2)) 378.0) (* (/ (Math|Vector|Distance(Vector) (.location (Utilities|Array|Get(acopy) _vinelocations _array_index_2)) (.location _array_element_2)) 378.0) (Variables|Default|GetVineZScale)))) true))
                        (Rendering|Material|SetDefaultCustomPrimitiveDataVector3 _instancevines_tree_hanging _returnvalue (Transformation|GetActorLocation))))))))))))))

(fn ConstructionScript
  (bind _returnvalue (Transformation|GetActorTransform))
  (bind _stream (Variables|Default|GetStream))
  (|CallParentFunction)
  (Math|Random|SetRandomStreamSeed (Variables|Default|GetStream) (Variables|Default|GetSeed))
  (Utilities|FlowControl|Sequence
    (:then_0
      (if (Editor|InEditor)
        (if (Variables|Default|GetEdit%20Mode)
          (Variables|Default|SetSavedTransform _returnvalue)
          (CallFunction|SpawnSupportVines)
          (elif (== (Variables|Default|GetSavedTransform) _returnvalue)
            _
            (else
              (Variables|Default|SetSavedTransform _returnvalue)
              (Variables|Default|SetEdit%20Mode true)
              (CallFunction|SpawnSupportVines))))))
    (:then_1
      (CallFunction|SpawnVines)
      (Default|SpawnVineInstances)
      (Utilities|FlowControl|Sequence
        (:then_0
          (if (Variables|Default|GetAddVineAddonMesh)
            (Math|Random|SetRandomStreamSeed _stream (Math|Conversions|ToInteger(Integer64) (+ (Variables|Default|GetSeed) (Math|Float|TruncatetoInteger64 (+ (.x (Transformation|GetActorLocation)) (.y (Transformation|GetActorLocation)))))))
            (if (>= (Math|Random|RandomFloatinRangefromStream _stream 0.0 1.0) 0.1)
              (bind _returnvalue_2 (AddComponent|NODE_AddStaticMeshComponent-0 :RelativeTransform (Math|Transform|MakeTransform "0, 0, 0")))
              (Components|StaticMesh|SetStaticMesh _returnvalue_2 (Utilities|Select :Index (Math|Random|RandomIntegerinRangefromStream _stream 0 1) :Option%200 (Variables|Default|GetAddonStaticmeshOption2) :Option%201 (Variables|Default|GetAddonStaticmeshOption1))))))
        (:then_1
          (CallFunction|SpawnSmallPlants))
        (:then_2
          (if (not (Variables|Default|GetSpawn%20Support_Vines))
            (Components|DestroyComponent (Variables|Default|GetConnectionsVines_Connector_HISM))
            (Components|DestroyComponent (Variables|Default|GetConnectionsVines_HISM))))))))
