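# Editor utility Blueprint in the Asteria tree foliage folder that builds UDIM textures from individual Texture2D assets.
# Its CreateUDIM function passes an output path, source textures, block coordinates and flags for keeping existing settings and check-out-and-save straight into Make UDIM Texture from Texture2Ds.
# The function does not return the created texture, and the event graph and construction script hold only empty default nodes.
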EditorUtilityBlueprint /Game/Environments/Asteria/Foliage/Trees/Blueprint_Parent/Utility_MakeUDIM
graphs [
 CreateUDIM {nodes [
   n1 = CreateUDIM!() -> (
    then
    OutputPathName:String
    SourceTextures:Array of Texture 2D Object References
    BlockCoords:Array of Int Point Structures
    bKeepExistingSettings:bool=~
    bCheckOutAndSave:bool=~
   ) {is=title K2Node_FunctionEntry}
   > Make UDIM Texture from Texture2Ds!(
    execute
    OutputPathName:String=@n1
    SourceTextures:Array of Texture 2D Object References=@n1.SourceTextures
    BlockCoords:Array of Int Point Structures=@n1.BlockCoords
    bKeepExistingSettings:bool=@n1.bKeepExistingSettings
    bCheckOutAndSave:bool=@n1.bCheckOutAndSave
   ) -> (then, ReturnValue:Texture 2D Object Reference)
  ]}
 EventGraph {nodes [
   Event BeginPlay()
   Event ActorBeginOverlap()
   Event Tick!() -> (OutputDelegate, then, DeltaSeconds:float)
  ]}
 UserConstructionScript {nodes [Construction Script()]}
]
variables []
